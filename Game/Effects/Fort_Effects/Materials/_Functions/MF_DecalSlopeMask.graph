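# MF_DecalSlopeMask is a material function that masks decal opacity by surface slope.
# It rebuilds a per-pixel surface normal from the cross product of world position's screen-space derivatives, substituting a constant up vector on mobile.
# That normal is dotted with the SlopeDirection parameter, which defaults to up, then sharpened with CheapContrast.
# The result multiplies the Opacity input to form the output.
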MaterialFunction /Game/Effects/Fort_Effects/Materials/_Functions/MF_DecalSlopeMask
graph {nodes [
  n1 = WorldPosition()
  n2 = CrossProduct(A=Normalize(VectorInput=DDX(Value=@n1.XYZ)), B=Normalize(VectorInput=DDY(Value=@n1.XYZ)))
  n3 = DotProduct(
   A=ShadingPathSwitch(Default=@n2, Deferred=@n2, Mobile=Constant3Vector() {constant=<rgba 0 0 1 0>}) {desc=Migrated from FeatureLevelSwitch}
   B=VectorParameter() {parameterName=SlopeDirection, defaultValue=<rgba 0 0 1 0>}.RGB
  )
  n4 = Constant() {r=1}
  n5 = FunctionInput() {bUsePreviewValueAsDefault=true, inputName=Opacity, inputType=FunctionInput_Scalar, previewValue=<1 0 0 1>}
  FunctionOutput(None=Multiply(
    A=MaterialFunctionCall(In=@n3, Contrast=@n4) {materialFunction=&CheapContrast}.Result
    B=@n5
   ))
 ]}
