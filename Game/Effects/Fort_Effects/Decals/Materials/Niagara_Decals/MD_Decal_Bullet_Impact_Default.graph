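# Translucent deferred decal material for default bullet impacts, in the Niagara decal set.
# Opacity combines decal color alpha, a mask texture and a scalar.
# Glow multiplies decal color by tiled noise, a radial sphere mask and the same mask, while the normal adds crack detail to a base map.
# Particle random offsets the noise and crack UVs to vary each impact.
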Material /Game/Effects/Fort_Effects/Decals/Materials/Niagara_Decals/MD_Decal_Bullet_Impact_Default
graph {
 properties {
  bAutomaticallySetUsageInEditor=false
  blendMode=BLEND_Translucent
  materialDomain=MD_DeferredDecal
  pixelDepthOffsetMode=PDOM_Legacy
  shadingModel=MSM_Unlit
  usedShadingModels=Unlit
 }
 nodes [
  DecalDerivative()
  LightVector()
  LightVector()
  MaterialFunctionCall!(Opacity) {desc=Based on this, materialFunction=/Game/Effects/Fort_Effects/Materials/_Functions/MF_DecalSlopeMask}
  MaterialFunctionCall() {desc=Based on this, materialFunction=/Game/Weapons/Prototype/SprayGun/Materials/MF_DecalNormal}
  n1 = DecalColor()
  n2 = TextureSampleParameter2D() {parameterName=Opacity Mask, samplerType=SAMPLERTYPE_LinearColor, texture=/Game/Effects/Fort_Effects/Textures/Decal/T_Decal_Default_A}
  Multiply(A=@n1.A, B=Multiply(A=@n2.G, B=ScalarParameter() {defaultValue=0.66, parameterName=Opacity}))
  n3 = TextureCoordinate()
  n4 = Add(A=@n3, B=ParticleRandom())
  n5 = ScalarParameter() {defaultValue=1, parameterName=Emissive Tile}
  n6 = Multiply(
   A=TextureSampleParameter2D(UVs=Multiply(A=@n4, B=@n5)) {parameterName=Emissive Mask, texture=/Game/Global/Textures/Default/Noise/T_Tiling_Noise_01}.G
   B=ScalarParameter() {defaultValue=10, parameterName=Emissive Intensity}
  )
  n7 = Power(Base=SphereMask(A=@n3, B=Constant() {r=0.5}, Radius=ScalarParameter() {defaultValue=0.5, parameterName=Emissive Radius}, Hardness=Constant()))
  Multiply(A=@n1.RGB, B=Multiply(A=Multiply(A=@n6, B=@n7), B=Power(Base=@n2.G)))
  n8 = TextureSampleParameter2D() {parameterName=Normal, samplerType=SAMPLERTYPE_Normal, texture=/Game/Effects/Fort_Effects/Textures/Decal/T_Decal_Default_N}
  n9 = TextureCoordinate()
  n10 = ParticleRandom()
  n11 = TextureSampleParameter2D(UVs=Multiply(A=Add(A=@n9, B=@n10), B=ScalarParameter() {defaultValue=1, parameterName=Normal Detail Tile})) {parameterName=NormalDetail, samplerType=SAMPLERTYPE_Normal, texture=/Game/Effects/Fort_Effects/Textures/Decal/T_Decal_Detail_Cracks}
  n12 = ScalarParameter() {defaultValue=0.1, parameterName=Normal Detail Intensity}
  n13 = Multiply(
   A=@n11.RGB
   B=MaterialFunctionCall(X=@n12, Y=@n12, Z=Constant()) {materialFunction=&MakeFloat3}.Result
  )
  n14 = ScalarParameter() {defaultValue=1, parameterName=Normal Intensity}
  n15 = Constant() {r=1}
  Normalize(VectorInput=Multiply(
    A=Add(A=@n8.RGB, B=@n13) {constB=0}
    B=MaterialFunctionCall(X=@n14, Y=@n14, Z=@n15) {materialFunction=&MakeFloat3}.Result
   ))
  ScalarParameter() {parameterName=Base Color}
  ScalarParameter() {parameterName=Metallic}
  ScalarParameter() {parameterName=Specular}
  ScalarParameter() {defaultValue=1, parameterName=Roughness}
 ]
}
